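# UI translucent material for adaptive text colour that responds to the blurred Slate post buffer and button state.
# Named reroutes A and C derive idle colour from CoreUI BackplateA and BackplateC functions; B desaturates, eases, inverts and rounds a rerouted input into binary black or white.
# Each feeds a ModularButton state lerp supplying fixed hovered, focused and pressed colours plus a shared disabled grey.
# A "Use Light" static switch chooses between two rerouted values.
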Material /FortUILibrary/Materials/Text/M_UI_AdaptiveText
graph {
 properties {
  bUsedWithStaticMesh=false
  blendMode=BLEND_Translucent
  materialDomain=MD_UI
  pixelDepthOffsetMode=PDOM_Legacy
  usedShadingModels=Default Lit
 }
 nodes [
  n1 = NamedRerouteUsage()
  Add(A=@n1, B=Constant() {r=0.3}) {constB=0.1}
  Multiply(A=@n1) {constB=0.5}
  n2 = LinearInterpolate(
   A=Constant() {r=1}
   B=Constant()
   Alpha=MaterialFunctionCall("Value (0-1)"=Desaturation(None=@n1)) {materialFunction=&MF_UI_EaseCurves}.EaseOutQuint
  )
  Multiply(A=@n2, B=@n2)
  NamedRerouteDeclaration(None=MaterialFunctionCall!("UVs (Optional)") {materialFunction=/Engine/Functions/UserInterface/GetSlatePost0}.LinearRGB) {name=Blur RGB, nodeColor=<rgba 0 0 0 1>}
  n3 = NamedRerouteUsage()
  n4 = Constant3Vector() {constant=<rgba 1 1 1 0>}
  n5 = NamedRerouteDeclaration(None=Constant3Vector() {constant=<rgba 0.231 0.231 0.231 1>}) {name=Disabled Color, nodeColor=<rgba 0 0 0 1>}
  NamedRerouteDeclaration(None=MaterialFunctionCall!(
    Idle=MaterialFunctionCall!(Post Buffer Blur=@n3) {materialFunction=/FortUILibrary/MaterialFunctions/CoreUI/MF_UI_CoreUI_BackplateA}.Result
    Idle Opacity
    Hovered=@n4
    Hovered Opacity
    Focused=@n4
    Focused Opacity
    Pressed=@n4
    Pressed Opacity
    Disabled=@n5
    Disabled Opacity
   ) {materialFunction=/FortUILibrary/Systems/ModularButton/MaterialFunctions/MF_UIKit_ButtonStatesBase_Lerp}.Color) {name=A, nodeColor=<rgba 0 0 0 1>}
  n6 = Round(None=@n2)
  NamedRerouteDeclaration(None=MaterialFunctionCall(Idle=Saturate(None=@n6), Hovered=@n4, Focused=@n4, Pressed=@n4, Disabled=@n5) {materialFunction=&MF_UIKit_ButtonStatesBase_Lerp}.Color) {name=B, nodeColor=<rgba 0 0 0 1>}
  n7 = NamedRerouteUsage()
  n8 = Constant3Vector() {constant=<rgba 0.0048 0.004 0.0116 1>}
  NamedRerouteDeclaration(None=MaterialFunctionCall(
    Idle=MaterialFunctionCall!(Post Buffer Blur=@n7) {materialFunction=/FortUILibrary/MaterialFunctions/CoreUI/MF_UI_CoreUI_BackplateC}.Result
    Hovered=@n8
    Focused=@n8
    Pressed=@n8
    Disabled=@n5
   ) {materialFunction=&MF_UIKit_ButtonStatesBase_Lerp}.Color) {name=C, nodeColor=<rgba 0 0 0 1>}
  StaticSwitchParameter(True=NamedRerouteUsage(), False=NamedRerouteUsage()) {defaultValue=true, parameterName=Use Light}
  VertexColor()
 ]
}
