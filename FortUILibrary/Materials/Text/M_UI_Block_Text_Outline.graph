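# Translucent UI material, named as a block text outline, that builds an alpha mask driven by modular button states.
# Base and selectable state lerp functions chain together, with the base Color output feeding both the color and opacity inputs of the selectable stage.
# The resulting opacity is multiplied by a saturated MF_UI_Circle fill on non-tiled 9-slice UVs, then by vertex alpha.
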Material /FortUILibrary/Materials/Text/M_UI_Block_Text_Outline
graph {
 properties {
  bUsedWithStaticMesh=false
  blendMode=BLEND_Translucent
  materialDomain=MD_UI
  pixelDepthOffsetMode=PDOM_Legacy
  usedShadingModels=Default Lit
 }
 nodes [
  n1 = MaterialFunctionCall!(Idle, Idle Opacity, Hovered, Hovered Opacity, Focused, Focused Opacity, Pressed, Pressed Opacity, Disabled, Disabled Opacity) {materialFunction=/FortUILibrary/Systems/ModularButton/MaterialFunctions/MF_UIKit_ButtonStatesBase_Lerp}
  n2 = MaterialFunctionCall(
   UVs=MaterialFunctionCall() {materialFunction=&GetUserInterfaceUV}."9-Slice UV (No Tiling)"
   Size=Constant()
   GlowMax=Constant() {r=2}
  ) {materialFunction=&MF_UI_Circle}
  Multiply(
   A=VertexColor().A
   B=Multiply(
    A=MaterialFunctionCall!(
     Base States=@n1.Color
     Base States Opacity=@n1.Color
     Selected_Idle
     Selected_Idle Opacity
     Selected_Hovered
     Selected_Hovered Opacity
     Selected_Focused
     Selected_Focused Opacity
     Selected_Pressed
     Selected_Pressed Opacity
     Selected_Disabled
     Selected_Disabled Opacity
    ) {materialFunction=/FortUILibrary/Systems/ModularButton/MaterialFunctions/MF_UIKit_ButtonStatesSelectable_Lerp}.Opacity
    B=Saturate(None=@n2.Fill)
   ) {constB=0.25}
  )
 ]
}
